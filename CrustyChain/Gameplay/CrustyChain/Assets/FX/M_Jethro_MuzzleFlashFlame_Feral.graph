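# Unlit, two-sided translucent material for Jethro's Feral muzzle flash flame, usable on Niagara sprites, mesh particles and skeletal meshes.
# A shape texture is modulated by panning, distorted noise, with Is Radial Core switching shape channel and planar or radial UVs.
# Dynamic parameters drive noise amount, erosion, Tibanna gas curve color selection and HDR intensity.
# Particle color tints emissive; opacity adds non-mobile depth fade and two-sided view-angle fade.
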Material /CrustyChain/Gameplay/CrustyChain/Assets/FX/M_Jethro_MuzzleFlashFlame_Feral
graph {
 properties {
  bUseTranslucencyVertexFog=false
  bUsedWithMeshDeformer=true
  bUsedWithMeshParticles=true
  bUsedWithNiagaraMeshParticles=true
  bUsedWithNiagaraSprites=true
  bUsedWithSkeletalMesh=true
  blendMode=BLEND_Translucent
  pixelDepthOffsetMode=PDOM_Legacy
  shadingModel=MSM_Unlit
  twoSided=true
  usedShadingModels=Unlit
 }
 nodes [
  n1 = TextureSampleParameter2D() {group=FlameShape, parameterName=MuzzleShape, texture=&/T_Jethro_MuzzleFlash_M}
  n2 = LinearInterpolate(
   A=LinearInterpolate(A=@n1.R, B=@n1.R, Alpha=VertexColor().R)
   B=LinearInterpolate(A=@n1.B, B=@n1.B, Alpha=DynamicParameter() {paramNames [Noise Amount, Erosion, Core Selection, Param4]}.Core Selection)
   Alpha=ScalarParameter() {parameterName=Is Radial Core}
  ) {constAlpha=0}
  n3 = TextureCoordinate()
  n4 = AppendVector(
   A=ScalarParameter() {defaultValue=1, group=Noise, parameterName=U Distort Freq, sortPriority=5}
   B=ScalarParameter() {defaultValue=1, group=Noise, parameterName=V Distort Freq, sortPriority=6}
  )
  n5 = TextureSampleParameter2D(UVs=Panner(Coordinate=Multiply(A=@n3, B=@n4)) {speedY=1}) {group=Noise, parameterName=Texture, sortPriority=4, texture=/Game/Weapons/FORT_Pistols/Effects/Muzzle/T_Cloud_d}
  n6 = TextureCoordinate()
  n7 = Multiply(A=ComponentMask(None=@n6) {r=true}, B=ScalarParameter() {defaultValue=6, parameterName=Noise_U_Scale})
  n8 = ScalarParameter() {parameterName=Nouse_U_Offset}
  n9 = Multiply(A=ComponentMask(None=@n6) {g=true}, B=ScalarParameter() {defaultValue=3, parameterName=Noise_V_Scale})
  n10 = ScalarParameter() {parameterName=Nouse_V_Offset}
  n11 = Add(
   A=Multiply(
    A=ComponentMask(None=@n5.RGB) {g=true, r=true}
    B=ScalarParameter() {defaultValue=1, group=Noise, parameterName=Distortion Amplitude, sortPriority=7}
   )
   B=AppendVector(A=Add(A=@n7, B=@n8), B=Add(A=@n9, B=@n10))
  )
  n12 = TextureCoordinate()
  n13 = AppendVector(
   A=ScalarParameter() {defaultValue=1, group=Noise, parameterName=U Distort Freq, sortPriority=5}
   B=ScalarParameter() {defaultValue=1, group=Noise, parameterName=V Distort Freq, sortPriority=6}
  )
  n14 = MaterialFunctionCall(Vector or UVs=Subtract(A=Multiply(A=@n12, B=@n13), B=Multiply(A=@n13) {constB=0.5})) {materialFunction=&VectorToRadialValue}
  n15 = ComponentMask(None=TextureSampleParameter2D(UVs=Panner(Coordinate=@n14.Radial Coordinates) {speedY=-1}) {group=Noise, parameterName=Texture, sortPriority=4, texture=&T_Cloud_d}.RGB) {g=true, r=true}
  n16 = ScalarParameter() {defaultValue=1, group=Noise, parameterName=Distortion Amplitude, sortPriority=7}
  n17 = TextureCoordinate()
  n18 = Multiply(A=ComponentMask(None=@n17) {r=true}, B=ScalarParameter() {defaultValue=6, parameterName=Noise_U_Scale})
  n19 = ScalarParameter() {parameterName=Nouse_U_Offset}
  n20 = Multiply(A=ComponentMask(None=@n17) {g=true}, B=ScalarParameter() {defaultValue=3, parameterName=Noise_V_Scale})
  n21 = ScalarParameter() {parameterName=Nouse_V_Offset}
  n22 = Subtract(A=AppendVector(A=Add(A=@n18, B=@n19), B=Add(A=@n20, B=@n21)), B=Multiply(A=AppendVector(A=@n18, B=@n20)) {constB=0.5})
  n23 = MaterialFunctionCall(Vector or UVs=Add(
    A=Multiply(A=@n15, B=@n16)
    B=MaterialFunctionCall(Vector or UVs=@n22) {materialFunction=&VectorToRadialValue}.Radial Coordinates
   )) {materialFunction=&VectorToRadialValue}
  n24 = ScalarParameter() {parameterName=Is Radial Core}
  n25 = Power(
   Base=TextureSampleParameter2D(UVs=LinearInterpolate(A=@n11, B=@n23.Radial Coordinates, Alpha=@n24) {constAlpha=0}) {group=Noise, parameterName=Noise, texture=/Game/Global/Textures/Default/Tileable_Patterns/T_Tornado_Tile}.R
   Exp=ScalarParameter() {defaultValue=1.5, group=Noise, parameterName=Noise Exp}
  ) {constExponent=1.5}
  n26 = DynamicParameter() {paramNames [Noise Amount, Erosion, Core Selection, Param4]}
  n27 = LinearInterpolate(A=@n2, B=Multiply(A=@n2, B=Saturate(None=@n25)), Alpha=@n26.Noise Amount)
  n28 = Power(Base=Clamp(None=Subtract(A=@n27, B=@n26.Erosion)), Exp=ScalarParameter() {defaultValue=1.5, parameterName=HDR Color Contrast})
  n29 = DynamicParameter() {parameterIndex=1, defaultValue=<rgba 0 1 1 1>, paramNames [Curve Select, HDR Intensity, Param3, Param4]}
  n30 = ParticleColor()
  n31 = Multiply(
   A=LinearInterpolate(
    A=@n28
    B=MaterialFunctionCall!(CurveTime=@n28, Curve Select=@n29.Curve Select) {materialFunction=/CrustyChain/Gameplay/CrustyChain/Assets/FX/Curves/Jethro_TibannaGasColorSelection}.Result
    Alpha=ScalarParameter() {defaultValue=1, parameterName=Tibanna Gas Color Amount}
   ) {constA=1}
   B=@n30.RGB
  )
  MaterialFunctionCall(Emissive=MaterialFunctionCall(LDR Input=Saturate(None=@n31), HDR Tint=@n31, HDR Intensity Multiplier=@n29.HDR Intensity) {materialFunction=&DeriveHDRfromLDR}.RemappedRange) {materialFunction=&MF_FX_EmissiveAdjustments}
  n32 = ScalarParameter() {defaultValue=20, parameterName=FadeDistance}
  n33 = Multiply(A=@n30.A, B=@n28)
  n34 = Abs(None=DotProduct(A=CameraVectorWS(), B=VertexNormalWS())) {desc=Make this work on two-sided materials}
  n35 = ScalarParameter() {parameterName="FadeSensitivity (optional) (S)"}
  Multiply(A=ShadingPathSwitch(Default=DepthFade(Opacity=@n33, FadeDistance=@n32), Mobile=@n33), B=SmoothStep(Value=Power(Base=@n34, Exp=@n35)))
 ]
}
