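# Niagara system for Jethro's weapon overheat effect, layering a glowing weapon mesh, cooldown smoke sprites and cones, ambient steam and muzzle flares.
# User.WeaponInt picks the overlay mesh and indexes per-weapon arrays of offsets and sizes.
# User.HasMedallion gates steam, flare and muzzle glow spawning and boosts the mesh material.
# User.OverHeat scales opacity of all but the cooldown smoke and cones, while User.ADS dims the flares and suppresses steam scaling.
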NiagaraSystem /CrustyChain/Gameplay/CrustyChain/Assets/FX/NS_Jethro_OverHeat_Base
dependencies {
 usedDataInterfaces [
  &NiagaraDataInterfaceMeshRendererInfo
  &NiagaraDataInterfaceCurve
  &NiagaraDataInterfaceArrayFloat3
  &NiagaraDataInterfaceSpriteRendererInfo
  &NiagaraDataInterfaceColorCurve
  &NiagaraDataInterfaceArrayFloat
  &NiagaraDataInterfaceVectorCurve
  &NiagaraDataInterfaceArrayFloat2
 ]
 usedDynamicInputs []
 usedModules [
  &self:SystemSpawnScript.SystemScriptSource.SystemScriptGraph.NiagaraNodeAssignment_2.SetVariables_#1
  &SystemState
 ]
}
emitters {
 Ambient_Steam {
  emitterSpawnScript [/SetVariables(Emitter.SmokeDepthFade=10) {isSetParameters=true}]
  emitterUpdateScript [
   &EmitterState()
   &SpawnRate(
    SpawnRate=&MakeCustomFloatFromBool(Bool=$User.HasMedallion, True Float=2, False Float=0)
    Use Spawn Probability=false
   )
  ]
  particleSpawnScript [
   &InitializeParticle(
    Color=<rgba 1 0.9187 0.7288 0.7>
    Lifetime Max=1.2
    Lifetime Min=0.8
    Lifetime Mode=::NewEnumerator1
    Mass Mode=::NewEnumerator3
    Position Offset=&Add_Vector(
     A=$System.BaseOffset
     B=&Multiply_VectorByFloat(
      Vector=&SelectVectorFromArray(
       Direct Array Index=$User.WeaponInt
       Direct Array Mode=::NewEnumerator1
       Vector Selection Array=DataInterface() {propertyValues {FloatData [<0 -20 0>, <0 -20 0>, <0 -25 0>, <0 -8 0>, <0 -10 0>, <0 -18 0>, <0 -12 0>, <0 -50 0>, <0 -10 0>, <0 -15 0>, <0 -34 0>]}}
      )
      Float=&RandomRangeFloat(Minimum=0, Maximum=1)
     )
    )
    Position Offset Coordinate Space=::Local
    Sprite Rotation Mode=::NewEnumerator2
    Sprite Size Mode=::NewEnumerator3
    Uniform Sprite Size=&Multiply_Float(
     A=&SelectFloatFromArray(
      Direct Array Index=$User.WeaponInt
      Float Selection Array=DataInterface() {propertyValues {FloatData [0.6,0.6,0.8,0.5,0.6,0.65,0.5,0.8,0.4,0.45,0.8]}}
     )
     B=&RandomRangeFloat(Minimum=18, Maximum=30)
    )
    UsePositionOffset=true
    Write Material Random=true
   )
   &CameraOffset(Camera Offset Amount=2)
   &DynamicMaterialParameters(
    Index 0 Param 1=1
    Module.Index0Float0Write=true
    Module.Index0Float1Write=true
    Module.Index0Float2Write=true
    Module.Index0Float3Write=true
   )
   &AddVelocity(Rotation Coordinate Space=::World, Velocity=&RandomRangeVector(Minimum=<0 0 7.5>, Maximum=<0 0 30>))
  ]
  particleUpdateScript [
   &ParticleState()
   &Drag(Drag=3, Use Linear Drag=true, Use Rotational Drag=true)
   &SolveForcesAndVelocity(Clamp Velocity=false, Limit Acceleration=false)
   &SubUVAnimation(
    Random Start Frame=true
    "Sprite Renderer "=DataInterface() {propertyValues {SpriteRenderer=&self:Ambient_Steam_0.NiagaraSpriteRendererProperties_0}}
    Start Frame Offset=5
    UseEndFrame=false
    UseStartFrame=false
   )
   &ScaleColor(
    Scale Alpha=&FloatFromCurve(
     FloatCurve=DataInterface() {propertyValues {Curve {keys [
         {arriveTangent=9.0725, interpMode=RCIM_Cubic, leaveTangent=9.0725, tangentMode=RCTM_User, time=0, value=0}
         {arriveTangent=-2.5249, interpMode=RCIM_Cubic, leaveTangent=-2.5249, tangentMode=RCTM_User, time=0.1408, value=1}
         {arriveTangent=-0.0405, interpMode=RCIM_Cubic, leaveTangent=-0.0405, tangentMode=RCTM_User, time=1, value=0}
        ]}}}
     Scale Curve=&Multiply_Float(A=$User.OverHeat, B=0.7)
    )
    Scale RGB=&MakeVectorFromLinearColorRGB(Color=&RandomRangeLinearColor(
      Minimum=&ColorFromCurve(ColorCurve=DataInterface() {propertyValues {
         AlphaCurve {keys [<at 0 1 RCIM_Cubic>]}
         BlueCurve {keys [<at 0.3401 0.3802 RCIM_Cubic>,<at 1 1 RCIM_Cubic>]}
         GreenCurve {keys [<at 0.3401 0.3658 RCIM_Cubic>,<at 1 0.5357 RCIM_Cubic>]}
         RedCurve {keys [<at 0.3401 0.7552 RCIM_Cubic>,<at 1 0.376 RCIM_Cubic>]}
        }})
      Maximum=&ColorFromCurve(ColorCurve=DataInterface() {propertyValues {
         AlphaCurve {keys [<at 0 1 RCIM_Cubic>]}
         BlueCurve {keys [<at 0.3388 0.3265 RCIM_Cubic>,<at 1 1 RCIM_Cubic>]}
         GreenCurve {keys [<at 0.3388 0.408 RCIM_Cubic>,<at 1 0.6409 RCIM_Cubic>]}
         RedCurve {keys [<at 0.3388 0.7552 RCIM_Cubic>,<at 1 0.2738 RCIM_Cubic>]}
        }})
     ))
    ScaleA=true
    ScaleRGB=true
   )
   &ScaleSpriteSize(
    Uniform Curve Scale=&MakeCustomFloatFromBool(Bool=$User.ADS, True Float=0, False Float=1)
    Uniform Curve Sprite Scale=DataInterface() {propertyValues {Curve {keys [
        {arriveTangent=0.5974, interpMode=RCIM_Cubic, leaveTangent=0.5974, tangentMode=RCTM_User, time=0, value=1}
        {arriveTangent=0.1871, interpMode=RCIM_Cubic, leaveTangent=0.1871, tangentMode=RCTM_User, time=1, value=1.5}
       ]}}}
   )
  ]
  rendererClasses [&NiagaraSpriteRendererProperties]
  renderers [{rendererClass=&NiagaraSpriteRendererProperties}]
 }
 CooldownSmoke {
  emitterUpdateScript [
   &EmitterState()
   &SpawnRate(SpawnRate=35, Use Spawn Probability=false)
  ]
  particleSpawnScript [
   &InitializeParticle(
    Color=<rgba 0.224 0.224 0.224 0.17>
    Lifetime Max=0.7
    Lifetime Min=0.2
    Lifetime Mode=::NewEnumerator1
    Position Offset=$System.BaseOffset
    Position Offset Coordinate Space=::Local
    Sprite Rotation Mode=::NewEnumerator2
    Sprite Size Mode=::NewEnumerator1
    Uniform Sprite Size Max=48
    Uniform Sprite Size Min=38
    UsePositionOffset=true
    Write Material Random=true
   )
   &PartitionParticles(Number Of Partitions=3)
   &InitialMeshOrientation(
    Mesh Orientation Mode=::NewEnumerator7
    Rotation=&Lerp_Vector(A=<0 0 0>, B=<0 0.5 0>, Alpha=&MakeFloatFromInt(Int=$Particles.PartitionParticles.Partition))
    Use Rotation Vector=true
   )
   &AddVelocity(
    Cone Angle=35
    Cone Axis=&SelectVectorFromArray(
     Direct Array Index=$Particles.PartitionParticles.Partition
     Vector Selection Array=DataInterface() {propertyValues {FloatData [<0 1 0>, <-1 0 0>, <1 0 0>]}}
    )
    Module.UseConeFalloff=false
    Module.UseVelDistribution=false
    Override Seed=false
    Velocity Mode=::NewEnumerator2
    Velocity Speed=&RandomRangeFloat(Minimum=160, Maximum=175)
   )
  ]
  particleUpdateScript [
   &ParticleState()
   &DynamicMaterialParameters(
    Index 0 Param 1=&FloatFromCurve(FloatCurve=DataInterface() {propertyValues {Curve {keys [
         {arriveTangent=3.0057, interpMode=RCIM_Cubic, leaveTangent=3.0057, tangentMode=RCTM_User, time=0, value=0}
         {arriveTangent=0.5254, interpMode=RCIM_Cubic, leaveTangent=0.5254, tangentMode=RCTM_User, time=1, value=1}
        ]}}})
    Module.Index0Float0Write=true
    Module.Index0Float1Write=true
    Module.Index0Float2Write=true
    Module.Index0Float3Write=true
   )
   &Drag(Drag=7, Use Linear Drag=true, Use Rotational Drag=true)
   &InheritSourceMovement(
    Applied Position Scale=&Multiply_VectorByFloat(
     Vector=<1 1 1>
     Float=&FloatFromCurve(FloatCurve=DataInterface() {propertyValues {Curve {keys [<at 0 1>,<at 1 0.5>]}}})
    )
    ApplyForce=false
    ApplyPosition=true
    ApplyPositionFalloff=false
   )
   &SolveForcesAndVelocity(Clamp Velocity=false, Limit Acceleration=false)
   &SubUVAnimation(
    Random Start Frame=true
    "Sprite Renderer "=DataInterface() {propertyValues {SpriteRenderer=&self:CooldownSmoke_0.NiagaraSpriteRendererProperties_0}}
    SubUV Lookup Index Scale=0.5
    UseEndFrame=false
    UseStartFrame=false
   )
   &ScaleSpriteSize(Uniform Curve Sprite Scale=DataInterface() {propertyValues {Curve {keys [
        {arriveTangent=3.6786, interpMode=RCIM_Cubic, leaveTangent=3.6786, tangentMode=RCTM_User, time=0, value=0}
        {interpMode=RCIM_Cubic, tangentMode=RCTM_User, time=1, value=1}
       ]}}})
   &ScaleColor(
    Scale Alpha=&FloatFromCurve(FloatCurve=DataInterface() {propertyValues {Curve {keys [
         {interpMode=RCIM_Cubic, tangentMode=RCTM_User, time=0, value=1}
         {arriveTangent=-1.7859, interpMode=RCIM_Cubic, leaveTangent=-1.7859, tangentMode=RCTM_User, time=1, value=0}
        ]}}})
    Scale RGB=&MakeVectorFromLinearColorRGB(Color=&ColorFromCurve(ColorCurve=DataInterface() {propertyValues {
        AlphaCurve {keys [<at 0 1 RCIM_Cubic>]}
        BlueCurve {keys [<at 0.1617 0.3818 RCIM_Cubic>,<at 0.5201 1 RCIM_Cubic>]}
        GreenCurve {keys [<at 0.1617 0.2917 RCIM_Cubic>,<at 0.5201 1 RCIM_Cubic>]}
        RedCurve {keys [<at 0.1617 1 RCIM_Cubic>,<at 0.5201 1 RCIM_Cubic>]}
       }}))
    ScaleA=true
    ScaleRGB=true
   ) {module=ScaleColor001}
  ]
  rendererClasses [&NiagaraSpriteRendererProperties]
  renderers [{rendererClass=&NiagaraSpriteRendererProperties}]
 }
 CooldownSmokeCones {
  emitterUpdateScript [
   &EmitterState()
   &SpawnRate(SpawnRate=10, Use Spawn Probability=false)
   &SpawnRate(SpawnGroup=1, SpawnRate=10, Use Spawn Probability=false) {module=SpawnRate001}
   &SpawnRate(SpawnGroup=2, SpawnRate=10, Use Spawn Probability=false) {module=SpawnRate002}
  ]
  particleSpawnScript [
   &InitializeParticle(
    Color=<rgba 1 1 1 0.225>
    Lifetime Max=0.2
    Lifetime Min=0.1
    Lifetime Mode=::NewEnumerator1
    Mesh Scale Max=<0.22 0.25 0.25>
    Mesh Scale Min=<0.18 0.2 0.2>
    Mesh Scale Mode=::NewEnumerator2
    Position Offset=$System.BaseOffset
    UsePositionOffset=true
    Write Material Random=true
   )
   &InitialMeshOrientation(
    Mesh Orientation Mode=::NewEnumerator7
    Rotation=&SelectVectorFromArray(
     Direct Array Index=&SpawnGroup()
     Vector Selection Array=DataInterface() {propertyValues {FloatData [<0 0 0>, <0 0.5 0>, <0 0 0.25>]}}
    )
    Use Rotation Vector=true
   )
  ]
  particleUpdateScript [
   &ParticleState()
   &ScaleMeshSize(Scale Factor=&Multiply_VectorByFloat(
     Vector=&MakeVector(
      X=&FloatFromCurve(FloatCurve=DataInterface() {propertyValues {Curve {keys [<at 0 0.5>,<at 1 1.5>]}}})
      Y=&FloatFromCurve(FloatCurve=DataInterface() {propertyValues {Curve {keys [<at 0 0.5>,<at 1 1.5>]}}})
      Z=&FloatFromCurve(FloatCurve=DataInterface() {propertyValues {Curve {keys [<at 0 0.5>,<at 1 1.5>]}}})
     )
     Float=1
    ))
   &ScaleColor(
    Scale Alpha=&FloatFromCurve(FloatCurve=DataInterface() {propertyValues {Curve {keys [
         {interpMode=RCIM_Cubic, tangentMode=RCTM_User, time=0, value=0}
         {arriveTangent=-3.0848, interpMode=RCIM_Cubic, leaveTangent=-3.0848, tangentMode=RCTM_User, time=0.2, value=1}
         {arriveTangent=0.0189, interpMode=RCIM_Cubic, leaveTangent=0.0189, tangentMode=RCTM_User, time=1, value=0}
        ]}}})
    ScaleA=true
    ScaleRGB=true
   )
   &DynamicMaterialParameters(
    Index 0 Param 3=1
    Index 0 Param 4=1
    Module.Index0Float0Write=true
    Module.Index0Float1Write=true
    Module.Index0Float2Write=true
    Module.Index0Float3Write=true
   )
  ]
  rendererClasses [&NiagaraMeshRendererProperties]
  renderers [{rendererClass=&NiagaraMeshRendererProperties, rendererIndex=0}]
 }
 MedallionFlare {
  emitterUpdateScript [
   &EmitterState()
   &SpawnRate(
    SpawnGroup=1
    SpawnRate=&MakeCustomFloatFromBool(Bool=$User.HasMedallion, True Float=6, False Float=0)
    Use Spawn Probability=false
   )
  ]
  particleSpawnScript [
   &InitializeParticle(
    Color=<rgba 1.5 0.3849 0 0.4>
    Lifetime=0.7
    Mesh Scale=&SelectVectorFromArray(
     Direct Array Index=$User.WeaponInt
     Direct Array Mode=::NewEnumerator1
     Vector Selection Array=DataInterface() {propertyValues {FloatData [
        <0.1725 0.115 0.115>
        <0.1725 0.115 0.115>
        <0.165 0.135 0.135>
        <0.075 0.1 0.1>
        <0.1 0.135 0.135>
        <0.1 0.155 0.155>
        <0.1 0.12 0.12>
        <0.17 0.1 0.1>
        <0.08 0.1 0.1>
        <0.12 0.08 0.08>
        <0.2 0.15 0.15>
       ]}}
    )
    Mesh Scale Mode=::NewEnumerator4
    Position Offset=&SelectVectorFromArray(
     Direct Array Index=$User.WeaponInt
     Direct Array Mode=::NewEnumerator1
     Vector Selection Array=DataInterface() {propertyValues {FloatData [<2 53 9.25>, <2.25 49 9.25>, <0 86 7>, <0 35.5 11>, <0 67 9>, <0 51 2.25>, <0 75 9.5>, <0 108 8.75>, <0 42.5 6.5>, <0 40 9.6>, <0 95 8>]}}
    )
    Position Offset Coordinate Space=::Local
    Sprite Rotation Mode=::NewEnumerator2
    Sprite Size Mode=::NewEnumerator3
    Sprite UV Mode=::NewEnumerator3
    Uniform Sprite Size=&Multiply_Float(
     A=&SelectFloatFromArray(
      Direct Array Index=$User.WeaponInt
      Direct Array Mode=::NewEnumerator1
      Float Selection Array=DataInterface() {propertyValues {FloatData [0.8,0.8,1,0.7,1,1,0.7,1,0.7,0.6,1]}}
     )
     B=100
    )
    UsePositionOffset=true
    Write Material Random=true
   )
   &InitialMeshOrientation(
    Mesh Orientation Mode=::NewEnumerator2
    Orientation Vector=<0 1 0>
    Rotation=&RandomRangeVector(Minimum=<-1 0 0>, Maximum=<1 0 0>)
    Use Rotation Vector=true
   )
   &DynamicMaterialParameters(
    Index 0 Param 1=10
    Index 1 Param 1=0.2
    Index 1 Param 2=3
    Module.Index0Float0Write=true
    Module.Index0Float1Write=true
    Module.Index0Float2Write=true
    Module.Index0Float3Write=true
    Module.Index1Float0Write=true
    Module.Index1Float1Write=true
    Module.Index1Float2Write=true
    Module.Index1Float3Write=true
    Write Parameter Index 1=::NewEnumerator1
   )
   &CameraOffset(Camera Offset Amount=-5)
  ]
  particleUpdateScript [
   &ParticleState()
   &ScaleMeshSize(Scale Factor=&VectorFromCurve(VectorCurve=DataInterface() {propertyValues {
       XCurve {keys [
         {arriveTangent=3.3837, interpMode=RCIM_Cubic, leaveTangent=3.3837, tangentMode=RCTM_User, time=0, value=0}
         {arriveTangent=0.1318, interpMode=RCIM_Cubic, leaveTangent=0.1318, tangentMode=RCTM_User, time=1, value=1.2}
        ]}
       YCurve {keys [
         {arriveTangent=2.479, interpMode=RCIM_Cubic, leaveTangent=2.479, tangentMode=RCTM_User, time=0, value=0}
         {arriveTangent=0.0687, interpMode=RCIM_Cubic, leaveTangent=0.0687, tangentMode=RCTM_User, time=1, value=1}
        ]}
       ZCurve {keys [
         {arriveTangent=2.479, interpMode=RCIM_Cubic, leaveTangent=2.479, tangentMode=RCTM_User, time=0, value=0}
         {arriveTangent=0.0687, interpMode=RCIM_Cubic, leaveTangent=0.0687, tangentMode=RCTM_User, time=1, value=1}
        ]}
      }}))
   &ScaleSpriteSize(Uniform Curve Sprite Scale=DataInterface() {propertyValues {Curve {
       defaultValue=0
       keys [
        {arriveTangent=1.6882, interpMode=RCIM_Cubic, leaveTangent=1.6882, tangentMode=RCTM_User, time=0, value=0}
        {interpMode=RCIM_Cubic, tangentMode=RCTM_User, time=1, value=1}
       ]
      }}})
   &ScaleColor(
    Scale Alpha=&FloatFromCurve(
     FloatCurve=DataInterface() {propertyValues {Curve {keys [
         {arriveTangent=48.0273, interpMode=RCIM_Cubic, leaveTangent=48.0273, tangentMode=RCTM_User, time=0, value=0}
         <at 0.05 1 RCIM_Cubic>
         {interpMode=RCIM_Cubic, leaveTangent=-3.1778, time=1, value=0}
        ]}}}
     Scale Curve=&Multiply_Float(A=$User.OverHeat, B=&MakeCustomFloatFromBool(Bool=$User.ADS, True Float=0.1, False Float=1))
    )
    Scale RGB=&MakeVectorFromLinearColorRGB(Color=&ColorFromCurve(ColorCurve=DataInterface() {propertyValues {
        AlphaCurve {keys []}
        BlueCurve {keys [<at 0.1026 0.2859 RCIM_Cubic>,<at 0.959 0.5 RCIM_Cubic>]}
        GreenCurve {keys [<at 0.1026 0.639 RCIM_Cubic>,<at 0.959 0.5 RCIM_Cubic>]}
        RedCurve {keys [<at 0.1026 1 RCIM_Cubic>,<at 0.959 0.5 RCIM_Cubic>]}
       }}))
    ScaleA=true
    ScaleRGB=true
   )
   /SetVariables(
    Particles.MeshConeCameraOffset=2
    Particles.MeshConeColor:LinearColor=&MultiplyLinearColor(
     Linear Color=<rgba 2 2 2 0.47>
     Scale Factor=&MakeVector4FromVectorAndFloat(
      "Vector (RGB)"=&MakeVectorFromLinearColorRGB(Color=&ColorFromCurve(ColorCurve=DataInterface() {propertyValues {
          AlphaCurve {keys [<at 0 1>]}
          BlueCurve {keys [
            <at 0.0313 0>
            <at 0.1131 0.0507>
            {arriveTangent=1.2499, interpMode=RCIM_Cubic, leaveTangent=1.2499, time=0.3888, value=1}
            {arriveTangent=1.2499, interpMode=RCIM_Cubic, leaveTangent=1.2499, time=0.6644, value=1}
            <at 0.8726 1>
           ]}
          GreenCurve {keys [
            <at 0.0313 0.1575>
            <at 0.1131 0.2973>
            {arriveTangent=0.7324, interpMode=RCIM_Cubic, leaveTangent=0.7324, time=0.3888, value=0.0452}
            {arriveTangent=0.7324, interpMode=RCIM_Cubic, leaveTangent=0.7324, time=0.6644, value=0.6571}
            <at 0.8726 0.8536>
           ]}
          RedCurve {keys [
            <at 0.0313 2>
            <at 0.1131 1>
            {arriveTangent=-0.5623, interpMode=RCIM_Cubic, leaveTangent=-0.5623, time=0.3888, value=0.0473}
            {arriveTangent=-0.5623, interpMode=RCIM_Cubic, leaveTangent=-0.5623, time=0.6644, value=0}
            <at 0.8726 0.5729>
           ]}
         }}))
      "Float (Alpha)"=&FloatFromCurve(
       FloatCurve=DataInterface() {propertyValues {Curve {keys [
           {arriveTangent=48.0273, interpMode=RCIM_Cubic, leaveTangent=48.0273, tangentMode=RCTM_User, time=0, value=0}
           {arriveTangent=-1.5073, interpMode=RCIM_Cubic, leaveTangent=-1.5073, tangentMode=RCTM_User, time=0.05, value=1}
           {arriveTangent=0.2056, interpMode=RCIM_Cubic, leaveTangent=0.2056, tangentMode=RCTM_User, time=1, value=0}
          ]}}}
       Scale Curve=&FloatFromCurve(
        FloatCurve=DataInterface() {propertyValues {Curve {defaultValue=1, keys [<at 0 0 RCIM_Cubic>, {tangentMode=RCTM_User, time=0.5, value=1}]}}}
        CurveIndex=$User.OverHeat
        Scale Curve=&MakeCustomFloatFromBool(Bool=$User.ADS, True Float=0.25, False Float=1)
       )
      )
     )
    )
    Particles.MeshConeDynm:Vector4f=<0.7 0.8 2 0>
    Particles.MeshConeDynm1:Vector4f=<0.5 1 1 1>
    Particles.MeshConePosition:NiagaraPosition=&AddVectorToPosition(
     Position=$Engine.Emitter.SimulationPosition
     Vector=&SelectVectorFromArray(
      Direct Array Index=$User.WeaponInt
      Direct Array Mode=::NewEnumerator1
      Vector Selection Array=DataInterface() {propertyValues {FloatData [<2 55 9>, <2.25 51.5 9.5>, <0 87 7>, <0 37 11>, <0 69 9>, <0 52.35 2.25>, <0 76 9.5>, <0 110 8.75>, <0 43.5 6.5>, <0 41.25 9.6>, <0 96.5 8>]}}
     )
    )
   ) {isSetParameters=true}
   /SetVariables(
    Particles.MeshMuzzle:Vector3f=<0.09 0.09 0.2>
    Particles.MeshMuzzleCameraOffset=1
    Particles.MeshMuzzleColor:LinearColor=&MultiplyLinearColor(
     Linear Color=<rgba 1.5 1.5 1.5 0.1>
     Scale Factor=&MakeVector4FromVectorAndFloat(
      "Vector (RGB)"=&MakeVectorFromLinearColorRGB(Color=&ColorFromCurve(ColorCurve=DataInterface() {propertyValues {
          AlphaCurve {keys [<at 0 1>]}
          BlueCurve {keys [<at 0.0438 0>, {arriveTangent=1.7689, interpMode=RCIM_Cubic, time=0.2921, value=1}, <at 0.5727 1 RCIM_Cubic>, <at 0.8414 1>]}
          GreenCurve {keys [
            <at 0.0438 0.2519>
            {arriveTangent=0.7168, interpMode=RCIM_Cubic, time=0.2921, value=0.0452}
            {arriveTangent=1.4718, interpMode=RCIM_Cubic, leaveTangent=1.4718, time=0.5727, value=0.6571}
            <at 0.8414 0.8536>
           ]}
          RedCurve {keys [
            <at 0.0438 1>
            {arriveTangent=-1.7689, interpMode=RCIM_Cubic, time=0.2921, value=0.0473}
            {arriveTangent=0.957, interpMode=RCIM_Cubic, leaveTangent=0.957, time=0.5727, value=0}
            <at 0.8414 0.5729>
           ]}
         }}))
      "Float (Alpha)"=&FloatFromCurve(
       FloatCurve=DataInterface() {propertyValues {Curve {keys [
           {arriveTangent=48.0273, interpMode=RCIM_Cubic, leaveTangent=48.0273, tangentMode=RCTM_User, time=0, value=0}
           {arriveTangent=-1.5073, interpMode=RCIM_Cubic, leaveTangent=-1.5073, tangentMode=RCTM_User, time=0.05, value=1}
           {arriveTangent=0.2056, interpMode=RCIM_Cubic, leaveTangent=0.2056, tangentMode=RCTM_User, time=1, value=0}
          ]}}}
       Scale Curve=&FloatFromCurve(
        FloatCurve=DataInterface() {propertyValues {Curve {defaultValue=1, keys [<at 0 0 RCIM_Cubic>, {tangentMode=RCTM_User, time=0.5, value=1}]}}}
        CurveIndex=$User.OverHeat
        Scale Curve=&MakeCustomFloatFromBool(Bool=$User.ADS, True Float=0.1, False Float=1)
       )
      )
     )
    )
    Particles.MeshMuzzleDynm:Vector4f=<1 1 15 1>
    Particles.MeshMuzzleDynm1:Vector4f=<1 1 1 1>
    Particles.MeshMuzzlePosition:NiagaraPosition=&AddVectorToPosition(
     Position=$Engine.Emitter.SimulationPosition
     Vector=&SelectVectorFromArray(
      Direct Array Index=$User.WeaponInt
      Direct Array Mode=::NewEnumerator1
      Vector Selection Array=DataInterface() {propertyValues {FloatData [<2 48 16>, <2.25 45 16>, <0 80 14>, <0 30 18>, <0 62 16>, <0 45 5>, <0 70 16>, <0 103 16>, <0 36.5 13>, <0 36 16>, <0 90 15>]}}
     )
    )
    Particles.MeshMuzzleQuat:Quat4f=&MultiplyQuaternion(
     Quaternion A=&MakeQuaternion(Quaternion From=3, Axis=<1 0 0>, Angle=0, From=<0 0 1>, To=<0 -1 2>, Angle Type=::NewEnumerator0, Normalize Vectors=false)
     Quaternion B=&MakeQuaternion(
      Quaternion From=0
      Axis=<0 0 1>
      Angle=&RandomRangeFloat(Minimum=-1, Maximum=1)
      From=?
      To=?
      Angle Type=::NewEnumerator1
      Normalize Vectors=false
     )
    )
   ) {isSetParameters=true}
  ]
  rendererClasses [&NiagaraSpriteRendererProperties, &NiagaraMeshRendererProperties]
  renderers [
   {rendererClass=&NiagaraSpriteRendererProperties}
   {rendererClass=&NiagaraMeshRendererProperties, rendererIndex=1}
   {rendererClass=&NiagaraMeshRendererProperties, rendererIndex=2}
  ]
 }
 MedallionMuzzleGlow {
  emitterUpdateScript [
   &EmitterState()
   &SpawnRate(
    SpawnGroup=1
    SpawnRate=&MakeCustomFloatFromBool(Bool=$User.HasMedallion, True Float=5, False Float=0)
    Use Spawn Probability=false
   )
  ]
  particleSpawnScript [
   &InitializeParticle(
    Color=<rgba 0.04 0.2153 0.4 0.4>
    Lifetime=0.7
    Mesh Scale=<0.05 0.05 0.22>
    Mesh Scale Mode=::NewEnumerator4
    Position Offset=&SelectVectorFromArray(
     Direct Array Index=$User.WeaponInt
     Direct Array Mode=::NewEnumerator1
     Vector Selection Array=DataInterface() {propertyValues {FloatData [<2 42 9.25>, <2 42 9.25>, <0 70 7>, <0 32 11>, <0 58 9>, <0 46 2.25>, <0 70 9.5>, <0 95 8.75>, <0 39 6.5>, <0 35 9.6>, <0 82 8>]}}
    )
    Sprite Size Mode=::NewEnumerator3
    Uniform Sprite Size=40
    UsePositionOffset=true
    Write Material Random=true
   )
   &CameraOffset(Camera Offset Amount=6)
   &SpriteFacingAndAlignment(WriteAlignment=true, WriteFacing=true)
   &DynamicMaterialParameters(
    Index 0 Param 1=2.5
    Index 0 Param 2=8
    Index 0 Param 3=1
    Index 0 Param 4=1
    Module.Index0Float0Write=true
    Module.Index0Float1Write=true
    Module.Index0Float2Write=true
    Module.Index0Float3Write=true
   )
  ]
  particleUpdateScript [
   &ParticleState()
   &ScaleSpriteSize(
    Scale Factor=&SelectVector2DFromArray(
     Direct Array Index=$User.WeaponInt
     Direct Array Mode=::NewEnumerator1
     Vector 2D Selection Array=DataInterface() {propertyValues {FloatData [<1.2 2>, <1.2 2>, <1.5 2.85>, <1 1.5>, <1.7 2.5>, <1.8 2.5>, <1.5 2.2>, <1.2 2.5>, <1.2 1.5>, <1 1.8>, <1.5 2.5>]}}
    )
    Scale Sprite Size Mode=::NewEnumerator2
   )
   &ScaleColor(
    Scale Alpha=&FloatFromCurve(
     FloatCurve=DataInterface() {propertyValues {Curve {keys [
         {arriveTangent=1.6135, interpMode=RCIM_Cubic, leaveTangent=1.6135, tangentMode=RCTM_User, time=0, value=0}
         {arriveTangent=-1.5073, interpMode=RCIM_Cubic, leaveTangent=-1.5073, tangentMode=RCTM_User, time=0.1, value=1}
         {arriveTangent=0.2056, interpMode=RCIM_Cubic, leaveTangent=0.2056, tangentMode=RCTM_User, time=1, value=0}
        ]}}}
     Scale Curve=&FloatFromCurve(
      FloatCurve=DataInterface() {propertyValues {Curve {defaultValue=1, keys [<at 0 0 RCIM_Cubic>, {tangentMode=RCTM_User, time=0.8, value=1}]}}}
      CurveIndex=$User.OverHeat
      Scale Curve=&MakeCustomFloatFromBool(Bool=$User.ADS, True Float=0.1, False Float=1)
     )
    )
    ScaleA=true
    ScaleRGB=false
   )
  ]
  rendererClasses [&NiagaraSpriteRendererProperties]
  renderers [{rendererClass=&NiagaraSpriteRendererProperties}]
 }
 WeaponMeshArray {
  emitterUpdateScript [
   &EmitterState()
   &SpawnRate(SpawnRate=10, Use Spawn Probability=false)
  ]
  particleSpawnScript [&InitializeParticle(
    Color=&MultiplyLinearColor(
     Linear Color=<rgba 1.4 0.4896 0.0018 1>
     Scale Factor=&MakeVector4FromVectorAndFloat("Vector (RGB)"=<1 1 1>, "Float (Alpha)"=$User.OverHeat)
    )
    Lifetime Max=1.25
    Lifetime Min=0.9
    Lifetime Mode=::NewEnumerator1
    Mesh Renderer Array Visibility Mode=::NewEnumerator1
    Mesh Renderer Info=DataInterface() {propertyValues {MeshRenderer=&self:WeaponMeshArray_0.NiagaraMeshRendererProperties_0}}
    Mesh Renderer Mesh Index:NiagaraInt32=$User.WeaponInt
    Mesh Scale Mode=::NewEnumerator3
    UsePositionOffset=true
    Write Material Random=true
   )]
  particleUpdateScript [
   &ParticleState()
   &ScaleColor(
    Scale Alpha=&FloatFromCurve(
     FloatCurve=DataInterface() {propertyValues {Curve {
        defaultValue=1
        keys [
         {arriveTangent=9.0725, interpMode=RCIM_Cubic, leaveTangent=9.0725, tangentMode=RCTM_User, time=0, value=0}
         {arriveTangent=0.0192, interpMode=RCIM_Cubic, leaveTangent=0.0192, tangentMode=RCTM_User, time=0.08, value=0.9847}
         {arriveTangent=-0.0265, interpMode=RCIM_Cubic, leaveTangent=-0.0265, tangentMode=RCTM_User, time=1, value=0}
        ]
       }}}
     Scale Curve=0.5
    )
    Scale RGB=&MakeVectorFromLinearColorRGB(Color=<rgba 1 0.5649 0 1>)
    ScaleA=true
    ScaleRGB=true
   )
   &DynamicMaterialParameters(
    Index 0 Param 1=&MakeCustomFloatFromBool(Bool=$User.HasMedallion, True Float=2, False Float=0.4)
    Module.Index0Float0Write=true
    Module.Index0Float1Write=true
    Module.Index0Float2Write=true
    Module.Index0Float3Write=true
   )
  ]
  rendererClasses [&NiagaraMeshRendererProperties]
  renderers [{rendererClass=&NiagaraMeshRendererProperties, rendererIndex=0}]
 }
}
summary {
 systemName=NS_Jethro_OverHeat_Base
 emitters [
  {emitterName=WeaponMeshArray, rendererClasses [&NiagaraMeshRendererProperties]}
  {emitterName=CooldownSmokeCones, rendererClasses [&NiagaraMeshRendererProperties]}
  {emitterName=CooldownSmoke, rendererClasses [&NiagaraSpriteRendererProperties]}
  {emitterName=Ambient_Steam, rendererClasses [&NiagaraSpriteRendererProperties]}
  {emitterName=MedallionFlare, rendererClasses [&NiagaraSpriteRendererProperties, &NiagaraMeshRendererProperties]}
  {emitterName=MedallionMuzzleGlow, rendererClasses [&NiagaraSpriteRendererProperties]}
 ]
 userVariables [
  User.ADS {type=NiagaraBool}
  User.HasMedallion {type=NiagaraBool}
  User.OverHeat {type=NiagaraFloat, defaultValue=1, description=""}
  User.WeaponInt {type=NiagaraInt32, defaultValue=3}
 ]
}
