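# Niagara system that draws faux-projectile bolt tracers for Jethro Feral hitscan shots, spawned from Niagara Data Channel reads.
# Two near-identical emitters filter by instigator, differing mainly in bolt length and travel easing.
# Each spawns ribbon point pairs that move from origin to target over distance/velocity lifetime, with curve-scaled tail stretch.
# VariantIndex picks light color and dynamic material parameters, and lights are gated by ribbon link order.
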NiagaraSystem /CrustyChain/Gameplay/CrustyChain/Assets/FX/NS_NDC_Jethro_ProjectilesHandler_HitScan_Feral
dependencies {
 usedDataInterfaces [
  &NiagaraDataInterfaceArrayColor
  &NiagaraDataInterfaceCurve
  &NiagaraDataInterfaceArrayFloat
  /Script/Niagara.NiagaraDataInterfaceDataChannelRead
 ]
 usedDynamicInputs []
 usedModules [
  &SystemState
  /Niagara/Modules/System/CompleteIfUnused
  &self:SystemSpawnScript.SystemScriptSource.SystemScriptGraph.NiagaraNodeAssignment_0.SetVariables_#1
 ]
}
emitters {
 Bolt_Tracer {
  emitterSpawnScript [/SetVariables(Emitter.BoltLength=500, Emitter.Projectile Velocity=12333.333) {isSetParameters=true, module=SetVariables_#2}]
  emitterUpdateScript [
   &EmitterState()
   &self:Spawn_NDC_SW(Instigator Filter=2, PositionType=2, Reader:NiagaraDataInterfaceDataChannelRead=$System.NDCReader, Spawn Count=2) {module=Spawn_NDC_SW}
  ]
  particleSpawnScript [
   &self:InitializeParticle_NDC_SW(
    Player Position:NiagaraPosition=&MakeVectorFromLinearColorRGB(Color=$NPC.FortniteNPC.PlayerPosition)
    Projectile Test Length=5000
   ) {module=InitializeParticle_NDC_SW}
   &InitializeParticle(
    Lifetime=&Divide_Float(A=$Particles.DistanceTraveled, B=$Emitter.Projectile Velocity)
    Ribbon Width=60
    Ribbon Width Mode=::NewEnumerator1
    UsePositionOffset=false
    Write Material Random=true
   )
   /SetVariables(Particles.LightColor=&SelectLinearColorFromArray(
     Array Sampling Mode=::NewEnumerator1
     Direct Array Index=$Particles.VariantIndex
     Color Selection Array=DataInterface() {propertyValues {ColorData [
        <rgba 1 0 0 1>
        <rgba 1 0.2 0 1>
        <rgba 1 0.5 0 1>
        <rgba 0.1269 1 0 1>
        <rgba 0 0.8333 1 1>
        <rgba 0 0.0833 1 1>
        <rgba 0.1623 0 1 1>
        <rgba 1 1 1 1>
       ]}}
     Random Seed=?
    )) {isSetParameters=true, module=SetVariables_#3}
  ]
  particleUpdateScript [
   &ParticleState()
   /Niagara/Modules/Update/Utility/IncrementOverTime()
   &self:FauxProjectileMovement_SW(
    Initial Value:NiagaraPosition=$Particles.OriginPosition
    Lerp COntroller=DataInterface() {propertyValues {Curve {defaultValue=0, keys [{tangentMode=RCTM_User, time=0, value=0}, {tangentMode=RCTM_User, time=1, value=1}]}}}
    Projectile Velocity:NiagaraFloat=$Emitter.Projectile Velocity
    Target Value:NiagaraPosition=$Particles.CurrentPosition
   ) {module=FauxProjectileMovement_SW}
   /SetVariables(Particles.TraveledAmount:NiagaraFloat=&VectorLength(Vector=&Subtract_Position(A=$Particles.DirectedPosition, B=$Particles.OriginPosition))) {isSetParameters=true, module=SetVariables_#4}
   /SetVariables(Particles.NormalizedTravel:NiagaraFloat=&Divide_Float(A=$Particles.TraveledAmount, B=$Particles.DistanceTraveled)) {isSetParameters=true, module=SetVariables_#5}
   /SetVariables(Particles.Position=&LerpPosition(
     A=$Particles.DirectedPosition
     B=&Subtract_Position(
      A=$Particles.DirectedPosition
      B=&Multiply_VectorByFloat(
       Vector=$Particles.VelocityDirection
       Float=&FloatFromCurve(
        FloatCurve=DataInterface() {propertyValues {Curve {keys [
            {arriveTangentWeight=0.0833, interpMode=RCIM_Cubic, leaveTangentWeight=0.1833, tangentMode=RCTM_User, time=0.1, value=1}
            {arriveTangentWeight=0.3056, interpMode=RCIM_Cubic, leaveTangentWeight=0.1111, tangentMode=RCTM_User, time=0.8, value=1}
            {arriveTangent=-15.5626, tangentMode=RCTM_User, time=1, value=0}
           ]}}}
        CurveIndex=$Particles.NormalizedTravel
        Scale Curve=&ClampFloat(Float=$Particles.TraveledAmount, Min=$Particles.TraveledAmount, Max=$Emitter.BoltLength)
       )
      )
     )
     Alpha=$Particles.RibbonLinkOrder
    )) {isSetParameters=true, module=SetVariables_#6}
   &DynamicMaterialParameters(
    Index 0 Param 1=&RemapRange(
     Input Value To Remap=&MakeFloatFromInt(Int=$Particles.VariantIndex)
     Input Min=0
     Input Max=7
     Output Min=0
     Output Max=1
    )
    Index 0 Param 2=&SelectFloatFromArray(Direct Array Index=$Particles.VariantIndex, Float Selection Array=DataInterface() {propertyValues {FloatData [1,1,0.1,0.5,3,3,3,1]}})
    Module.Index0Float0Write=true
    Module.Index0Float1Write=true
    Module.Index0Float2Write=false
    Module.Index0Float3Write=false
   )
   /SetVariables(
    Particles.AttenuationRadius:NiagaraFloat=&MakeCustomFloatFromBool(
     Bool=&SetBoolByFloatComparison(A=$Particles.RibbonLinkOrder, B=0, Comparison Type=::NewEnumerator7)
     True Float=1000
     False Float=0
    )
    Particles.LightEnabled:NiagaraBool=&SetBoolByFloatComparison(A=$Particles.RibbonLinkOrder, B=0, Comparison Type=::NewEnumerator7)
   ) {isSetParameters=true, module=SetVariables_#7}
  ]
  rendererClasses [
   &NiagaraRibbonRendererProperties
   &NiagaraLightRendererProperties
   &NiagaraComponentRendererProperties
  ]
  renderers [
   {rendererClass=&NiagaraRibbonRendererProperties}
   {rendererClass=&NiagaraLightRendererProperties, rendererIndex=1}
   {rendererClass=&NiagaraComponentRendererProperties, rendererIndex=2}
  ]
 }
 Bolt_Tracer001 {
  emitterSpawnScript [/SetVariables(Emitter.BoltLength=3000, Emitter.Projectile Velocity=12333.333) {isSetParameters=true, module=SetVariables_#2}]
  emitterUpdateScript [
   &EmitterState()
   &self:Spawn_NDC_SW(Instigator Filter=1, PositionType=2, Reader:NiagaraDataInterfaceDataChannelRead=$System.NDCReader, Spawn Count=2) {module=Spawn_NDC_SW}
  ]
  particleSpawnScript [
   &self:InitializeParticle_NDC_SW(
    Player Position:NiagaraPosition=&MakeVectorFromLinearColorRGB(Color=$NPC.FortniteNPC.PlayerPosition)
    Projectile Test Length=5000
   ) {module=InitializeParticle_NDC_SW}
   &InitializeParticle(
    Lifetime=&Divide_Float(A=$Particles.DistanceTraveled, B=$Emitter.Projectile Velocity)
    Ribbon Width=60
    Ribbon Width Mode=::NewEnumerator1
    UsePositionOffset=false
    Write Material Random=true
   )
   /SetVariables(Particles.LightColor=&SelectLinearColorFromArray(
     Array Sampling Mode=::NewEnumerator1
     Direct Array Index=$Particles.VariantIndex
     Color Selection Array=DataInterface() {propertyValues {ColorData [
        <rgba 1 0 0 1>
        <rgba 1 0.2 0 1>
        <rgba 1 0.5 0 1>
        <rgba 0.1269 1 0 1>
        <rgba 0 0.8333 1 1>
        <rgba 0 0.0833 1 1>
        <rgba 0.1623 0 1 1>
        <rgba 1 1 1 1>
       ]}}
     Random Seed=?
    )) {isSetParameters=true, module=SetVariables_#3}
  ]
  particleUpdateScript [
   &ParticleState()
   &IncrementOverTime()
   &self:FauxProjectileMovement_SW(
    Initial Value:NiagaraPosition=$Particles.OriginPosition
    Lerp COntroller=DataInterface() {propertyValues {Curve {
       defaultValue=0
       keys [
        {arriveTangent=0.9583, interpMode=RCIM_Cubic, leaveTangent=0.9583, leaveTangentWeight=0.4449, tangentMode=RCTM_User, time=0, value=0}
        {arriveTangentWeight=0.4449, interpMode=RCIM_Cubic, tangentMode=RCTM_User, time=1, value=1}
       ]
      }}}
    Projectile Velocity:NiagaraFloat=$Emitter.Projectile Velocity
    Target Value:NiagaraPosition=$Particles.CurrentPosition
   ) {module=FauxProjectileMovement_SW}
   /SetVariables(Particles.TraveledAmount:NiagaraFloat=&VectorLength(Vector=&Subtract_Position(A=$Particles.DirectedPosition, B=$Particles.OriginPosition))) {isSetParameters=true, module=SetVariables_#4}
   /SetVariables(Particles.NormalizedTravel:NiagaraFloat=&Divide_Float(A=$Particles.TraveledAmount, B=$Particles.DistanceTraveled)) {isSetParameters=true, module=SetVariables_#5}
   /SetVariables(Particles.Position=&LerpPosition(
     A=$Particles.DirectedPosition
     B=&Subtract_Position(
      A=$Particles.DirectedPosition
      B=&Multiply_VectorByFloat(
       Vector=$Particles.VelocityDirection
       Float=&FloatFromCurve(
        FloatCurve=DataInterface() {propertyValues {Curve {keys [
            {arriveTangentWeight=0.0833, interpMode=RCIM_Cubic, leaveTangentWeight=0.1833, tangentMode=RCTM_User, time=0.1, value=1}
            {arriveTangentWeight=0.3056, interpMode=RCIM_Cubic, leaveTangentWeight=0.1111, tangentMode=RCTM_User, time=0.8, value=1}
            {arriveTangent=-15.5626, tangentMode=RCTM_User, time=1, value=0}
           ]}}}
        CurveIndex=$Particles.NormalizedTravel
        Scale Curve=&ClampFloat(Float=$Particles.TraveledAmount, Min=&Multiply_Float(A=$Emitter.BoltLength, B=0.25), Max=$Emitter.BoltLength)
       )
      )
     )
     Alpha=$Particles.RibbonLinkOrder
    )) {isSetParameters=true, module=SetVariables_#6}
   &DynamicMaterialParameters(
    Index 0 Param 1=&RemapRange(
     Input Value To Remap=&MakeFloatFromInt(Int=$Particles.VariantIndex)
     Input Min=0
     Input Max=7
     Output Min=0
     Output Max=1
    )
    Index 0 Param 2=&SelectFloatFromArray(Direct Array Index=$Particles.VariantIndex, Float Selection Array=DataInterface() {propertyValues {FloatData [1,1,0.1,0.5,3,3,3,1]}})
    Module.Index0Float0Write=true
    Module.Index0Float1Write=true
    Module.Index0Float2Write=false
    Module.Index0Float3Write=false
   )
   /SetVariables(
    Particles.AttenuationRadius:NiagaraFloat=&MakeCustomFloatFromBool(
     Bool=&SetBoolByFloatComparison(A=$Particles.RibbonLinkOrder, B=0, Comparison Type=::NewEnumerator7)
     True Float=1000
     False Float=0
    )
    Particles.LightEnabled:NiagaraBool=&SetBoolByFloatComparison(A=$Particles.RibbonLinkOrder, B=0, Comparison Type=::NewEnumerator7)
   ) {isSetParameters=true, module=SetVariables_#7}
  ]
  rendererClasses [
   &NiagaraRibbonRendererProperties
   &NiagaraLightRendererProperties
   &NiagaraComponentRendererProperties
  ]
  renderers [
   {rendererClass=&NiagaraRibbonRendererProperties}
   {rendererClass=&NiagaraLightRendererProperties, rendererIndex=1}
   {rendererClass=&NiagaraComponentRendererProperties, rendererIndex=2}
  ]
 }
}
summary {
 systemName=NS_NDC_Jethro_ProjectilesHandler_HitScan_Feral
 emitters [
  {
   emitterName=Bolt_Tracer
   rendererClasses [
    &NiagaraRibbonRendererProperties
    &NiagaraLightRendererProperties
    &NiagaraComponentRendererProperties
   ]
  }
  {
   emitterName=Bolt_Tracer001
   rendererClasses [
    &NiagaraRibbonRendererProperties
    &NiagaraLightRendererProperties
    &NiagaraComponentRendererProperties
   ]
  }
 ]
 userVariables [User.SpawnTracer {type=NiagaraBool, defaultValue=true}]
}
